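# Physics asset for the SweetLetter backpack cosmetic, defining capsule bodies and joints for its fluff pieces and hanging keyring.
# The pack body carries two capsules and anchors small base bodies for eight fluff pieces and the keyring through angularly free constraints.
# Each dyn fluff and keyring body hangs from its base with limited swing and zero twist limit.
# Four keys chain off the dynamic keyring with similar limits.
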PhysicsAsset /BRCosmetics/Accessories/FORT_Backpacks/Backpack_F_MED_SweetLetter/Meshes/F_MED_SweetLetter_Pack_Physics
bodies {
 base_fluff_3_l [{shapeType=Capsule, length=0.5, radius=0.5, shapeName=base_fluff_3_l_capsule, rotation=<rot 90 0 0>}]
 base_fluff_3_r [{shapeType=Capsule, length=0.5, radius=0.5, shapeName=base_fluff_3_r_capsule, rotation=<rot 90 0 0>}]
 base_fluff_4_l [{shapeType=Capsule, length=0.5, radius=0.5, shapeName=base_fluff_4_l_capsule, rotation=<rot 90 0 0>}]
 base_fluff_4_r [{shapeType=Capsule, length=0.5, radius=0.5, shapeName=base_fluff_4_r_capsule, rotation=<rot 90 0 0>}]
 base_fluff_5_l [{shapeType=Capsule, length=0.5, radius=0.5, shapeName=base_fluff_5_l_capsule, rotation=<rot 90 0 0>}]
 base_fluff_5_r [{shapeType=Capsule, length=0.5, radius=0.5, shapeName=base_fluff_5_r_capsule, rotation=<rot 90 0 0>}]
 base_fluff_6_l [{shapeType=Capsule, length=0.5, radius=0.5, shapeName=base_fluff_6_l_capsule, rotation=<rot 90 0 0>}]
 base_fluff_6_r [{shapeType=Capsule, length=0.5, radius=0.5, shapeName=base_fluff_6_r_capsule, rotation=<rot 90 0 0>}]
 base_keyring [{shapeType=Capsule, length=0.5, radius=0.5, shapeName=base_keyring_capsule, rotation=<rot 90 0 0>}]
 dyn_fluff_3_l [{shapeType=Capsule, length=2, radius=2, shapeName=dyn_fluff_3_l_capsule, center=<2 0 0>, rotation=<rot 90 0 0>}]
 dyn_fluff_3_r [{shapeType=Capsule, length=2, radius=2, shapeName=dyn_fluff_3_r_capsule, center=<2 0 0>, rotation=<rot 90 0 0>}]
 dyn_fluff_4_l [{shapeType=Capsule, length=2, radius=2, shapeName=dyn_fluff_4_l_capsule, center=<2 0 0>, rotation=<rot 90 0 0>}]
 dyn_fluff_4_r [{shapeType=Capsule, length=2, radius=2, shapeName=dyn_fluff_4_r_capsule, center=<2 0 0>, rotation=<rot 90 0 0>}]
 dyn_fluff_5_l [{shapeType=Capsule, length=2, radius=2, shapeName=dyn_fluff_5_l_capsule, center=<2 0 0>, rotation=<rot 90 0 0>}]
 dyn_fluff_5_r [{shapeType=Capsule, length=2, radius=2, shapeName=dyn_fluff_5_r_capsule, center=<2 0 0>, rotation=<rot 90 0 0>}]
 dyn_fluff_6_l [{shapeType=Capsule, length=2, radius=2, shapeName=dyn_fluff_6_l_capsule, center=<2 0 0>, rotation=<rot 90 0 0>}]
 dyn_fluff_6_r [{shapeType=Capsule, length=2, radius=2, shapeName=dyn_fluff_6_r_capsule, center=<2 0 0>, rotation=<rot 90 0 0>}]
 dyn_key_l [{shapeType=Capsule, length=5, radius=0.5, shapeName=dyn_key_l_capsule, center=<6.3522 0 0>, rotation=<rot 90 0 0>}]
 dyn_key_m_l [{shapeType=Capsule, length=5, radius=0.5, shapeName=dyn_key_m_l_capsule, center=<5.6273 0 0>, rotation=<rot 90 0 0>}]
 dyn_key_m_r [{shapeType=Capsule, length=5, radius=0.5, shapeName=dyn_key_m_r_capsule, center=<4.871 0 0>, rotation=<rot 90 0 0>}]
 dyn_key_r [{shapeType=Capsule, length=5, radius=0.5, shapeName=dyn_key_r_capsule, center=<5.6273 0 0>, rotation=<rot 90 0 0>}]
 dyn_keyring [{shapeType=Capsule, length=2, radius=1.3, shapeName=dyn_keyring_capsule, center=<2.1939 -0.845 0>}]
 pack [
  {shapeType=Capsule, length=4, radius=4, center=<0 -5.9582 -11.368>, rotation=<rot 90 0 0>}
  {shapeType=Capsule, length=4, radius=4, shapeName=pack_capsule, center=<0 -6.3589 -8.9937>, rotation=<rot 90 0 0>}
 ]
}
constraints [
 {
  bone1Name=base_fluff_3_l
  bone2Name=pack
  swing1LimitDegrees=45
  swing1Motion=Free
  swing2LimitDegrees=45
  swing2Motion=Free
  twistLimitDegrees=45
  twistMotion=Free
 }
 {bone1Name=dyn_fluff_3_l, bone2Name=base_fluff_3_l, swing1LimitDegrees=30, swing2LimitDegrees=30, twistLimitDegrees=0}
 {
  bone1Name=base_fluff_3_r
  bone2Name=pack
  swing1LimitDegrees=45
  swing1Motion=Free
  swing2LimitDegrees=45
  swing2Motion=Free
  twistLimitDegrees=45
  twistMotion=Free
 }
 {bone1Name=dyn_fluff_3_r, bone2Name=base_fluff_3_r, swing1LimitDegrees=30, swing2LimitDegrees=30, twistLimitDegrees=0}
 {
  bone1Name=base_fluff_4_l
  bone2Name=pack
  swing1LimitDegrees=45
  swing1Motion=Free
  swing2LimitDegrees=45
  swing2Motion=Free
  twistLimitDegrees=45
  twistMotion=Free
 }
 {bone1Name=dyn_fluff_4_l, bone2Name=base_fluff_4_l, swing1LimitDegrees=15, swing2LimitDegrees=15, twistLimitDegrees=0}
 {
  bone1Name=base_fluff_4_r
  bone2Name=pack
  swing1LimitDegrees=45
  swing1Motion=Free
  swing2LimitDegrees=45
  swing2Motion=Free
  twistLimitDegrees=45
  twistMotion=Free
 }
 {bone1Name=dyn_fluff_4_r, bone2Name=base_fluff_4_r, swing1LimitDegrees=30, swing2LimitDegrees=30, twistLimitDegrees=0}
 {
  bone1Name=base_fluff_5_l
  bone2Name=pack
  swing1LimitDegrees=45
  swing1Motion=Free
  swing2LimitDegrees=45
  swing2Motion=Free
  twistLimitDegrees=45
  twistMotion=Free
 }
 {bone1Name=dyn_fluff_5_l, bone2Name=base_fluff_5_l, swing1LimitDegrees=30, swing2LimitDegrees=30, twistLimitDegrees=0}
 {
  bone1Name=base_fluff_5_r
  bone2Name=pack
  swing1LimitDegrees=45
  swing1Motion=Free
  swing2LimitDegrees=45
  swing2Motion=Free
  twistLimitDegrees=45
  twistMotion=Free
 }
 {bone1Name=dyn_fluff_5_r, bone2Name=base_fluff_5_r, swing1LimitDegrees=30, swing2LimitDegrees=30, twistLimitDegrees=0}
 {
  bone1Name=base_fluff_6_l
  bone2Name=pack
  swing1LimitDegrees=45
  swing1Motion=Free
  swing2LimitDegrees=45
  swing2Motion=Free
  twistLimitDegrees=45
  twistMotion=Free
 }
 {bone1Name=dyn_fluff_6_l, bone2Name=base_fluff_6_l, swing1LimitDegrees=30, swing2LimitDegrees=30, twistLimitDegrees=0}
 {
  bone1Name=base_fluff_6_r
  bone2Name=pack
  swing1LimitDegrees=45
  swing1Motion=Free
  swing2LimitDegrees=45
  swing2Motion=Free
  twistLimitDegrees=45
  twistMotion=Free
 }
 {bone1Name=dyn_fluff_6_r, bone2Name=base_fluff_6_r, swing1LimitDegrees=30, swing2LimitDegrees=30, twistLimitDegrees=0}
 {
  bone1Name=base_keyring
  bone2Name=pack
  swing1LimitDegrees=45
  swing1Motion=Free
  swing2LimitDegrees=45
  swing2Motion=Free
  twistLimitDegrees=45
  twistMotion=Free
 }
 {bone1Name=dyn_keyring, bone2Name=base_keyring, swing1LimitDegrees=60, swing2LimitDegrees=20, twistLimitDegrees=0}
 {bone1Name=dyn_key_l, bone2Name=dyn_keyring, swing1LimitDegrees=40, swing2LimitDegrees=30, twistLimitDegrees=0}
 {bone1Name=dyn_key_m_l, bone2Name=dyn_keyring, swing1LimitDegrees=40, swing2LimitDegrees=15, twistLimitDegrees=0}
 {bone1Name=dyn_key_m_r, bone2Name=dyn_keyring, swing1LimitDegrees=40, swing2LimitDegrees=15, twistLimitDegrees=0}
 {bone1Name=dyn_key_r, bone2Name=dyn_keyring, swing1LimitDegrees=40, swing2LimitDegrees=30, twistLimitDegrees=0}
]
