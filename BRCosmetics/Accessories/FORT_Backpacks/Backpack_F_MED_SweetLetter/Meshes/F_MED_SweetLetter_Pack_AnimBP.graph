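# Animation Blueprint for the SweetLetter backpack cosmetic, adding secondary motion and skydiving pose adjustments.
# A RigidBody node with updated gravity override feeds chained AnimDynamics springs on fluff, strap and zipper bones.
# Additive bone transforms then rotate four fluff bones, weighted by the pawn's skydiving flag, and apply a fixed pack offset.
# The EventGraph updates gravity via a deprecated function and caches pawn, body mesh and gender, which the AnimGraph leaves unused.
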AnimBlueprint /BRCosmetics/Accessories/FORT_Backpacks/Backpack_F_MED_SweetLetter/Meshes/F_MED_SweetLetter_Pack_AnimBP
target_skeleton=&/F_MED_SweetLetter_Pack_Skeleton
graphs [
 AnimGraph {nodes [
   n1 = RigidBody!(ComponentPose=Local To Component(LocalPose=Input Pose()), Alpha, OverrideWorldGravity:Vector=Get GravityOverride(), ExternalForce) -> (Pose) {
    bOverrideWorldGravity=true
    overridePhysicsAsset=&/F_MED_SweetLetter_Pack_Physics
    alphaBoolBlend {blendOption=Linear}
    alphaScaleBiasClamp {bInterpResult=true, interpSpeedDecreasing=2, interpSpeedIncreasing=1}
    baseBoneRef {boneName=root}
    componentAppliedLinearAccClamp=<3000 3000 3000>
    componentLinearAccScale=<0.25 0.25 0.25>
    componentLinearVelScale=<0.25 0.25 0.25>
    overrideWorldGravity {z=-980}
    simSpaceSettings {
     maxAngularAcceleration=20
     maxAngularVelocity=1
     maxLinearAcceleration=300
     maxLinearVelocity=150
     velocityScaleZ=0
     worldAlpha=0.7
     externalLinearDragV=<0.2 0.2>
    }
   }
   n2 = Anim Dynamics - Bone: dyn_fluff_2_l!(ComponentPose=@n1, Alpha=1, GravityOverride:Vector="0,0,0", ExternalForce="0,0,0") -> (Pose) {
    angularDampingOverride=0.99
    angularSpringConstant=200
    bAngularSpring=true
    bOverrideAngularDamping=true
    bUseGravityOverride=true
    gravityScale=0.35
    simulationSpace=BoneRelative
    alphaBoolBlend {blendOption=Linear}
    boundBone {boneName=dyn_fluff_2_l}
    componentLinearAccScale=<0.2 0.2 0.2>
    componentLinearVelScale=<0.2 0.2 0.2>
    gravityOverride {z=-980}
    physicsBodyDefinitions [{
      boundBone {boneName=dyn_fluff_2_l}
      boxExtents=<10 10 10>
      constraintSetup {angularLimitsMax=<20 20 80>, angularLimitsMin=<-20 -20 -20>, angularTarget=<1 0 0>}
      localJointOffset=<2 0 1>
     }]
   }
   n3 = Anim Dynamics - Bone: dyn_fluff_1_l!(ComponentPose=@n2, Alpha=1, GravityOverride:Vector="0,0,0", ExternalForce="0,0,0") -> (Pose) {
    angularDampingOverride=0.9
    angularSpringConstant=200
    bAngularSpring=true
    bOverrideAngularDamping=true
    bUseGravityOverride=true
    gravityScale=0.35
    simulationSpace=BoneRelative
    alphaBoolBlend {blendOption=Linear}
    boundBone {boneName=dyn_fluff_1_l}
    componentLinearAccScale=<0.2 0.2 0.2>
    componentLinearVelScale=<0.2 0.2 0.2>
    gravityOverride {z=-980}
    physicsBodyDefinitions [{
      boundBone {boneName=dyn_fluff_1_l}
      boxExtents=<10 5 10>
      constraintSetup {angularLimitsMax=<20 20 20>, angularLimitsMin=<-20 -20 -20>, angularTarget=<1 0 0>}
      localJointOffset=<0 0 1>
     }]
   }
   n4 = Anim Dynamics - Bone: dyn_fluff_2_r!(ComponentPose=@n3, Alpha=1, GravityOverride:Vector="0,0,0", ExternalForce="0,0,0") -> (Pose) {
    angularDampingOverride=0.99
    angularSpringConstant=200
    bAngularSpring=true
    bOverrideAngularDamping=true
    bUseGravityOverride=true
    gravityScale=0.35
    simulationSpace=BoneRelative
    alphaBoolBlend {blendOption=Linear}
    boundBone {boneName=dyn_fluff_2_r}
    componentLinearAccScale=<0.2 0.2 0.2>
    componentLinearVelScale=<0.2 0.2 0.2>
    gravityOverride {z=-980}
    physicsBodyDefinitions [{
      boundBone {boneName=dyn_fluff_2_r}
      boxExtents=<10 10 10>
      constraintSetup {angularLimitsMax=<20 20 50>, angularLimitsMin=<-20 -20 -20>, angularTarget=<1 0 0>}
      localJointOffset=<2 0 1>
     }]
   }
   n5 = Anim Dynamics - Bone: dyn_fluff_1_r!(ComponentPose=@n4, Alpha=1, GravityOverride:Vector="0,0,0", ExternalForce="0,0,0") -> (Pose) {
    angularDampingOverride=0.9
    angularSpringConstant=200
    bAngularSpring=true
    bOverrideAngularDamping=true
    bUseGravityOverride=true
    gravityScale=0.35
    simulationSpace=BoneRelative
    alphaBoolBlend {blendOption=Linear}
    boundBone {boneName=dyn_fluff_1_r}
    componentLinearAccScale=<0.2 0.2 0.2>
    componentLinearVelScale=<0.2 0.2 0.2>
    gravityOverride {z=-980}
    physicsBodyDefinitions [{
      boundBone {boneName=dyn_fluff_1_r}
      boxExtents=<10 5 10>
      constraintSetup {angularLimitsMax=<20 20 20>, angularLimitsMin=<-20 -20 -20>, angularTarget=<1 0 0>}
      localJointOffset=<0.5 0 1>
     }]
   }
   n6 = Anim Dynamics - Bone: base_strap(ComponentPose=@n5, Alpha=1, ExternalForce="0,0,0") {
    angularDampingOverride=1
    angularSpringConstant=3000
    bAngularSpring=true
    bOverrideAngularDamping=true
    bUseGravityOverride=true
    gravityScale=0.35
    simulationSpace=BoneRelative
    alphaBoolBlend {blendOption=Linear}
    boundBone {boneName=base_strap}
    componentLinearAccScale=<1 1 1>
    componentLinearVelScale=<1 1 1>
    gravityOverride {z=-980}
    physicsBodyDefinitions [{
      boundBone {boneName=base_strap}
      boxExtents=<10 20 20>
      constraintSetup {angularLimitsMax=<0 30 15>, angularLimitsMin=<0 -80 -15>, angularTarget=<1 0 0>}
      localJointOffset=<4 0 0>
     }]
   }
   n7 = Anim Dynamics - Bone: dyn_zipper(ComponentPose=@n6, Alpha=1) {
    angularDampingOverride=0.9999
    angularSpringConstant=3000
    bAngularSpring=true
    bOverrideAngularDamping=true
    bUseGravityOverride=true
    gravityScale=0.35
    simulationSpace=BoneRelative
    alphaBoolBlend {blendOption=Linear}
    boundBone {boneName=dyn_zipper}
    componentLinearAccScale=<1 1 1>
    componentLinearVelScale=<1 1 1>
    gravityOverride {z=-980}
    physicsBodyDefinitions [{
      boundBone {boneName=dyn_zipper}
      boxExtents=<20 10 10>
      constraintSetup {angularLimitsMax=<0 100 15>, angularLimitsMin=<0 -30 -15>, angularTarget=<1 0 0>}
      localJointOffset=<2 0 0>
     }]
   }
   n8 = Get is skydiving() -> (is skydiving:double)
   n9 = "Transform (Modify) Bone - Bone: base_fluff_2_l"(ComponentPose=@n7, Alpha=@n8, Translation="0,0,0", Rotation="0,-50,0", Scale="1,1,1") {
    rotationMode=BMM_Additive
    rotationSpace=BCS_BoneSpace
    translationMode=BMM_Additive
    translationSpace=BCS_BoneSpace
    alphaBoolBlend {blendOption=Linear}
    boneToModify {boneName=base_fluff_2_l}
    rotation {yaw=-50}
   }
   n10 = "Transform (Modify) Bone - Bone: base_fluff_2_r"(ComponentPose=@n9, Alpha=@n8, Translation="0,0,0", Rotation="0,50,0", Scale="1,1,1") {
    rotationMode=BMM_Additive
    rotationSpace=BCS_BoneSpace
    translationMode=BMM_Additive
    translationSpace=BCS_BoneSpace
    alphaBoolBlend {blendOption=Linear}
    boneToModify {boneName=base_fluff_2_r}
    rotation {yaw=50}
   }
   n11 = "Transform (Modify) Bone - Bone: base_fluff_6_l"(ComponentPose=@n10, Alpha=@n8, Translation="0,0,0", Rotation="0,-10,0", Scale="1,1,1") {
    rotationMode=BMM_Additive
    rotationSpace=BCS_BoneSpace
    translationMode=BMM_Additive
    translationSpace=BCS_BoneSpace
    alphaBoolBlend {blendOption=Linear}
    boneToModify {boneName=base_fluff_6_l}
    rotation {yaw=-10}
   }
   n12 = "Transform (Modify) Bone - Bone: base_fluff_6_r"(ComponentPose=@n11, Alpha=@n8, Translation="0,0,0", Rotation="0,30,0", Scale="1,1,1") {
    rotationMode=BMM_Additive
    rotationSpace=BCS_BoneSpace
    translationMode=BMM_Additive
    translationSpace=BCS_BoneSpace
    alphaBoolBlend {blendOption=Linear}
    boneToModify {boneName=base_fluff_6_r}
    rotation {yaw=30}
   }
   n13 = "Transform (Modify) Bone - Bone: pack"(ComponentPose=@n12, Translation="0,-1.3,0", Rotation="0,0,5") {
    rotationMode=BMM_Additive
    rotationSpace=BCS_BoneSpace
    translationMode=BMM_Additive
    translationSpace=BCS_BoneSpace
    alphaBoolBlend {blendOption=Linear}
    boneToModify {boneName=pack}
    rotation {roll=5}
    translation {y=-1.3}
   }
   Output Pose(Result=Component To Local(ComponentPose=@n13))
   Get GravityOverride()
   Get GravityOverride()
  ]}
 EventGraph {nodes [
   Event BlueprintUpdateAnimation()
   > n1 = Gravity Override State Function DEPRECATED(CharacterPartAnimInstance=Self-Reference(), GameTimeInSeconds=GetGameTimeInSeconds(), GravityOverrideParams=Get Gravity Override Params())
   > Set GravityOverride(GravityOverride=@n1)
   > Set is skydiving(is skydiving:double="To Float (Boolean)"(InBool=Get bIsSkyDiving())) -> (Output_Get:double)
   Event OnCharacterCustomizationLoaded()
   > n2 = Set Owning Pawn(Owning Pawn:Pawn Object Reference=TryGetPawnOwner()) -> (Output_Get:Pawn Object Reference)
   > Is Valid(InputObject=@n2)
   > n3 = Sequence()
   > n4 = Cast To FortPlayerPawn(Object=Get Owning Pawn() -> (Owning Pawn:Pawn Object Reference))
   > n5 = Set Owning Player Pawn(Owning Player Pawn:Fort Player Pawn Object Reference=@n4) -> (Output_Get:Fort Player Pawn Object Reference)
   > Set BodyMesh(BodyMesh=GetSkeletalMeshForPartType(self=@n5))
   n6 = FindGender!(execute=@n3.then_1, Object:Object Reference=Get Owning Pawn()) -> (True:Exec, False:Exec, OutGender:EFortCustomGender Enum=Invalid)
   > Set isMale(isMale:bool="Equal (Enum)"(A:EFortCustomGender Enum=@n6, B:EFortCustomGender Enum=Male)) -> (Output_Get:bool)
   TryGetPawnOwner()
  ]}
]
variables [
 BodyMesh {type=SkeletalMeshComponent}
 Gravity Override Params {type=GravityOverrideParamsStruct}
 GravityOverride {type=Vector}
 is skydiving {type=number}
 isMale {type=boolean}
 Owning Pawn {type=Pawn, default=null}
 Owning Player Pawn {type=FortPlayerPawn}
]
